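; Sound library context for creature audio that plays an event's sound attached to a wildlife NPC pawn.
; Play lowers volume for creatures friendly to the local player, less so for Irwin VO-tagged events.
; When the creature is ridden, it feeds the sound parameters for enemy rider, fatigue from current energy, and normalized horizontal speed.
; The Convert component's converter pawn, when valid, supplies whether the rider is locally controlled.
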(fn Play
  (params (InEventData :type Sound%20Library%20Context%20Event%20Input%20Structure%20%28by%20ref%29))
  (bind (_sound _owner _eventtag) (Utilities|Struct|BreakSoundLibraryContextEventInput InEventData))
  (bind _asnpc_pawn_wildlife_parent (Utilities|Casting|CastToNPC_Pawn_Wildlife_Parent (Variables|SoundLibrary|GetActor)))
  (bind _returnvalue (Audio|SpawnSoundAttached :Sound _sound :AttachToComponent (Variables|/Script/Engine.Actor|GetRootComponent _asnpc_pawn_wildlife_parent) :AttachPointName "None" :LocationType "KeepRelativeOffset" :bStopWhenAttachedToDestroyed true :VolumeMultiplier (Utilities|Select :Index (Team|GetTeamAffiliationforActortoLocalPlayer _asnpc_pawn_wildlife_parent) :Friendly (Math|Float|SelectFloat 0.6 0.4 (GameplayTags|MatchesTag _eventtag "(TagName=\"SoundLibrary.Irwin.VO\")")) :Neutral 1.0 :Hostile 1.0) :PitchMultiplier 1.0 :StartTime 0.0 :bAutoDestroy true))
  (if (Variables|/Game/Athena/AI/Pawns/NPC/Parents/NPC_Pawn_Wildlife_Parent.NPC_Pawn_Wildlife_Parent_C|GetbIsBeingRidden _asnpc_pawn_wildlife_parent)
    (Audio|Parameter|SetBooleanParameter _returnvalue "IsRiddenByEnemy" (Utilities|Enum|Equal(Enum) (Team|GetTeamAffiliationforActortoLocalPlayer _asnpc_pawn_wildlife_parent) "Hostile") :!class /Script/AudioExtensions.AudioParameterControllerInterface)
    (Audio|Parameter|SetFloatParameter _returnvalue "FatigueLevel" (- 1.0 (Variables|/Game/Athena/AI/Pawns/NPC/Parents/NPC_Pawn_Wildlife_Parent.NPC_Pawn_Wildlife_Parent_C|GetCurrentEnergy _asnpc_pawn_wildlife_parent)) :!class /Script/AudioExtensions.AudioParameterControllerInterface)
    (Audio|Parameter|SetFloatParameter _returnvalue "Speed" (Math|Float|MapRangeClamped (Math|Vector2D|Vector2DLength (Math|Conversions|ToVector2D(Vector) (Variables|/Script/Engine.MovementComponent|GetVelocity (Pawn|GetMovementComponent _asnpc_pawn_wildlife_parent)))) 0.0 1.1e+03 0.0 1.0) :!class /Script/AudioExtensions.AudioParameterControllerInterface)
    (bind _convertcomponent (Variables|/Game/Athena/AI/Pawns/NPC/Parents/NPC_Pawn_Wildlife_Parent.NPC_Pawn_Wildlife_Parent_C|GetConvertComponent _asnpc_pawn_wildlife_parent)
      (:then
        (Utilities|IsValid (Convert|GetConverterPawn _convertcomponent)
          (:Is%20Valid
            (Audio|Parameter|SetBooleanParameter _returnvalue "IsLocal" (Pawn|IsLocallyControlled (Convert|GetConverterPawn _convertcomponent)) :!class /Script/AudioExtensions.AudioParameterControllerInterface))
          (:Is%20Not%20Valid
            _))
        (return (Utilities|Array|MakeArray _returnvalue) true))
      (:else
        _))
    (else
      _)))
